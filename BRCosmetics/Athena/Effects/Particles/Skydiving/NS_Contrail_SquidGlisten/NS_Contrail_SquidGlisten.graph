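# Skydiving contrail cosmetic for SquidGlisten that attaches meshes and ribbons to the character's skeletal mesh.
# Three "Birdarang" meshes fly in to attachment points, then blend into spinning orbits as the contrail setup's Dive Value rises, feeding red ribbon trails that fade with it.
# Time-of-day-tinted wind meshes and four burst-spawned ribbons, laid out along velocity by partition index, add airflow, with glide lean and front-end variants adjusting the look.
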NiagaraSystem /BRCosmetics/Athena/Effects/Particles/Skydiving/NS_Contrail_SquidGlisten/NS_Contrail_SquidGlisten
dependencies {
 usedDataInterfaces [
  &NiagaraDataInterfaceParticleRead
  &NiagaraDataInterfaceCurve
  &NiagaraDataInterfaceArrayFloat
  &NiagaraDataInterfaceArrayInt32
  &NiagaraDataInterfaceArrayFloat3
  &NiagaraDataInterfaceSkeletalMesh
 ]
 usedDynamicInputs [&FNTimeOfDayColor]
 usedModules [
  &self:SystemSpawnScript.SystemScriptSource.SystemScriptGraph.NiagaraNodeAssignment_0.SetVariables_#1
  &SystemState
  /Game/Effects/Niagara/Modules/Contrails/FN_Contrail_Setup
  &FN_Integrator
  &self:SystemSpawnScript.SystemScriptSource.SystemScriptGraph.NiagaraNodeAssignment_1.SetVariables_#2
 ]
}
emitters {
 Birdarangs {
  emitterUpdateScript [
   &EmitterState(Life Cycle Mode=::NewEnumerator1, Loop Behavior=::NewEnumerator1, Loop Duration Mode=::NewEnumerator1, UseLoopDelay=false)
   &SpawnBurst_Instantaneous(Spawn Count=3, Use Spawn Probability=false, UseLoopCountLimit=false)
  ]
  particleSpawnScript [
   &PartitionParticles(Number Of Partitions=3)
   &InitializeParticle(
    Color Mode=::NewEnumerator0
    Lifetime=0.85
    Mesh Scale Mode=::NewEnumerator3
    Mesh Uniform Scale=1.1
    UsePositionOffset=false
    Write Material Random=true
   )
  ]
  particleUpdateScript [
   &SolveForcesAndVelocity(Clamp Velocity=false, Limit Acceleration=false)
   &ParticleState(Kill Particles When Lifetime Has Elapsed=false)
   &SkeletalMeshLocation(
    Bone / Socket Index=&SelectIntFromArray(Direct Array Index=$Particles.PartitionParticles.Partition, Int Selection Array=DataInterface() {propertyValues {IntData [0,1,4]}})
    MaskSpawnGroup=false
    Mesh Sampling Type=::NewEnumerator5
    Position Sampling=::NewEnumerator2
    Sampled Position Offset=&FN_FrontEndVector3Modifier(
     Gameplay=&SelectVectorFromArray(
      Direct Array Index=$Particles.PartitionParticles.Partition
      Vector Selection Array=DataInterface() {propertyValues {FloatData [<16 -4 8>, <16 -4 -8>, <0 18 -10>]}}
     )
     Front End=&SelectVectorFromArray(
      Direct Array Index=$Particles.PartitionParticles.Partition
      Vector Selection Array=DataInterface() {propertyValues {FloatData [<16 -4 -4>, <16 -4 4>, <0 18 -10>]}}
     )
    )
    Skeletal Mesh=$System.Skeletal Mesh
    Socket Sampling Mode=::NewEnumerator5
   )
   &SkeletalMeshLocation(
    Bone / Socket Index=4
    Bone Sampling Mode=::NewEnumerator4
    MaskSpawnGroup=false
    Orientation Sampling=::NewEnumerator1
    Position Sampling=::NewEnumerator2
    Skeletal Mesh=$System.Skeletal Mesh
   ) {module=SkeletalMeshLocation001}
   &ScaleMeshSize(Scale Factor=&VectorFromFloat(Value=&FloatFromCurve(FloatCurve=DataInterface() {propertyValues {Curve {keys [
          {interpMode=RCIM_Cubic, leaveTangent=-0.1304, tangentMode=RCTM_Break, time=0, value=0}
          {arriveTangent=0.687, tangentMode=RCTM_User, time=0.3, value=1}
          <at 1 1>
         ]}}})))
   /SetVariables(Particles.PrevPos:NiagaraPosition=$Particles.Position) {isSetParameters=true}
   &RotateAroundPoint(
    Derive the Rotation Axis=false
    Radius:NiagaraFloat=&RandomRangeFloat(Minimum=70, Maximum=110)
    Rotation Axis:Vector3f=&RandomRangeVector(Minimum=<0 0 0>, Maximum=<1 1 1>)
    Rotation Center:NiagaraPosition=$Particles.SkeletalMeshLocation001.SampledPosition
    Rotation Phase:NiagaraFloat=&RandomRangeFloat(Minimum=-1, Maximum=1)
    Rotation Rate:NiagaraFloat=&SelectFloatFromArray(
     Array Sampling Mode=::NewEnumerator0
     Direct Array Index=?
     Float Selection Array=DataInterface() {propertyValues {FloatData [-1,1,1.5,-1.5]}}
     Random Seed=0
    )
    Set Particle Position=false
    X Axis:Vector3f=&RandomRangeVector(Minimum=<0 0 0>, Maximum=<1 1 1>)
   )
   /SetVariables(Particles.Position=&Lerp_Vector(
     A=&Lerp_Vector(
      A=Linked() {linkedVariable {name=Particles.SkeletalMeshLocation.SampledPosition, type=NiagaraPosition}}
      B=Linked() {linkedVariable {name=Output.RotateAroundPoint.OutputPosition, type=NiagaraPosition}}
      Alpha=&FloatFromCurve(
       FloatCurve=DataInterface() {propertyValues {Curve {keys [{interpMode=RCIM_Cubic, leaveTangent=3.1965, tangentMode=RCTM_Break, time=0, value=0}, <at 0.5 1>]}}}
       CurveIndex=$System.FN_Contrail_Setup.Dive Value
      )
     )
     B=&Lerp_Vector(
      A=&Multiply_Vector(A=Linked() {linkedVariable {name=Particles.SkeletalMeshLocation.SampledPosition, type=NiagaraPosition}}, B=<10 10 -5>)
      B=Linked() {linkedVariable {name=Particles.SkeletalMeshLocation.SampledPosition, type=NiagaraPosition}}
      Alpha=&FloatFromCurve(FloatCurve=DataInterface() {propertyValues {Curve {keys [
           {interpMode=RCIM_Cubic, leaveTangent=0.9438, tangentMode=RCTM_Break, time=0, value=0}
           {arriveTangent=1.035, tangentMode=RCTM_User, time=0.6, value=1}
          ]}}})
     )
     Alpha=&FloatFromCurve(FloatCurve=DataInterface() {propertyValues {Curve {keys [
          {interpMode=RCIM_Cubic, leaveTangent=-1.5069, tangentMode=RCTM_Break, time=0.6, value=1}
          {arriveTangent=-1.7839, tangentMode=RCTM_User, time=1, value=0}
         ]}}})
    )) {isSetParameters=true}
   &UpdateMeshOrientation(
    "Foward Vector (Mesh Space)":Vector3f=&Lerp_Vector(
     A=<1 0 0>
     B=&NormalizeVector(
      Vector To Normalize=&SubtractVectorFromPosition(Position=$Particles.PrevPos, Vector=Linked() {linkedVariable {name=Particles.Position, type=NiagaraPosition}})
      Safe Normalize=true
      Safe Fallback Threshold=0
      Safe Fallback Vector=<0 0 0.001>
     )
     Alpha=&FloatFromCurve(
      FloatCurve=DataInterface() {propertyValues {Curve {keys [{interpMode=RCIM_Cubic, leaveTangent=3.1965, tangentMode=RCTM_Break, time=0, value=0}, <at 0.5 1>]}}}
      CurveIndex=$System.FN_Contrail_Setup.Dive Value
     )
    )
    Initial Forward Vector:Vector3f=&SelectVectorFromArray(
     Direct Array Index=$Particles.PartitionParticles.Partition
     Vector Selection Array=DataInterface() {propertyValues {FloatData [<-1 0.2 0>, <-1 -0.2 0>, <-1 0 10>]}}
    )
    Orientation Method=::NewEnumerator3
   )
   &UpdateMeshOrientation(
    Delta Time:NiagaraFloat=$Engine.System.Age
    Rotation Rate=&Lerp_Float(
     A=&MakeCustomFloatFromBool(
      Bool=&SetBoolByFloatComparison(A=$Particles.NormalizedAge, B=0.7, Comparison Type=::NewEnumerator2)
      True Float=0
      False Float=5
     )
     B=4
     Alpha=&FloatFromCurve(
      FloatCurve=DataInterface() {propertyValues {Curve {keys [{interpMode=RCIM_Cubic, tangentMode=RCTM_Break, time=0, value=0}, {arriveTangent=3.139, tangentMode=RCTM_User, time=0.6, value=1}]}}}
      CurveIndex=$System.FN_Contrail_Setup.Dive Value
     )
    )
    Rotation Vector=<0 0 1>
   ) {module=UpdateMeshOrientation001}
  ]
  rendererClasses [&NiagaraSpriteRendererProperties, &NiagaraMeshRendererProperties]
  renderers [
   {rendererClass=&NiagaraSpriteRendererProperties}
   {rendererClass=&NiagaraMeshRendererProperties, rendererIndex=1}
  ]
 }
 Trails {
  emitterUpdateScript [
   &EmitterState()
   &SpawnParticlesFromOtherEmitter(
    Attribute Reader=DataInterface() {propertyValues {EmitterBinding {bindingMode=Other, emitterName=Birdarangs}}}
    Spawn Rate=30
    Use Maximum Age=false
    Use Minimum Age=false
   )
  ]
  particleSpawnScript [
   &InitializeParticle(
    Color=<rgba 0.5573 0.0088 0.002 0.35>
    Lifetime=0.35
    Ribbon Width=14
    Ribbon Width Mode=::NewEnumerator1
    UsePositionOffset=false
    Write Material Random=true
   )
   &SampleParticlesFromOtherEmitter(Mesh Orientation Sampling:ENiagara_DirectReadSamplingMode=::NewEnumerator0, Particle ID Sampling=::NewEnumerator1)
  ]
  particleUpdateScript [
   &ParticleState()
   &SolveForcesAndVelocity(Clamp Velocity=false, Limit Acceleration=false)
   &DynamicMaterialParameters(
    Index 0 Param 1=1
    Index 0 Param 2=0.7
    Module.Index0Float0Write=true
    Module.Index0Float1Write=true
    Module.Index0Float2Write=true
    Module.Index0Float3Write=true
   )
   &ScaleRibbonWidth(Ribbon Width Scale=&FloatFromCurve(FloatCurve=DataInterface() {propertyValues {Curve {keys [{interpMode=RCIM_Cubic, tangentMode=RCTM_Break, time=0, value=1}, {arriveTangent=-0.271, tangentMode=RCTM_User, time=1, value=0}]}}}))
   &ScaleColor(Scale Alpha=$System.FN_Contrail_Setup.Dive Value, ScaleA=true, ScaleRGB=true)
   &ScaleColor(
    Scale Alpha=&FloatFromCurve(
     FloatCurve=DataInterface() {propertyValues {Curve {keys [{interpMode=RCIM_Cubic, leaveTangent=-1.505, tangentMode=RCTM_Break, time=0, value=1}, <at 1 0>]}}}
     Scale Curve=&FN_FrontEndFloatModifier(Gameplay=1, Front End=0.7)
    )
    ScaleA=true
    ScaleRGB=true
   ) {module=ScaleColor001}
   /SetVariables(
    Particles.FillTrailColor:LinearColor=&Multiply_LinearColorByFloat(Linear Color=<rgba 0.1962 0.0005 0.0074 1>, Float=$System.FN_Contrail_Setup.Dive Value)
    Particles.FillTrailDynPar:Vector4f=&MakeVector4(
     X=&FloatFromCurve(FloatCurve=DataInterface() {propertyValues {Curve {keys [
          {interpMode=RCIM_Cubic, leaveTangent=0.052, tangentMode=RCTM_Break, time=0, value=1}
          {arriveTangent=0.0615, tangentMode=RCTM_User, time=1, value=0}
         ]}}})
     Y=1
     Z=1
     W=1
    )
    Particles.FillTrailRibbonWidth:NiagaraFloat=&Multiply_Float(A=$Particles.RibbonWidth, B=2.4)
   ) {isSetParameters=true}
   /SetVariables(
    Particles.AirTrailColor:LinearColor=&Multiply_LinearColorByFloat(Linear Color=<rgba 1 0.1649 0.1649 0.4>, Float=$System.FN_Contrail_Setup.Dive Value)
    Particles.AirTrailDynPar:Vector4f=&MakeVector4(X=-1.5727, Y=0, Z=1, W=4)
   ) {isSetParameters=true}
  ]
  rendererClasses [&NiagaraSpriteRendererProperties, &NiagaraRibbonRendererProperties]
  renderers [
   {rendererClass=&NiagaraSpriteRendererProperties}
   {rendererClass=&NiagaraRibbonRendererProperties, rendererIndex=1}
   {rendererClass=&NiagaraRibbonRendererProperties, rendererIndex=2}
   {rendererClass=&NiagaraRibbonRendererProperties, rendererIndex=3}
  ]
 }
 WindMesh001 {
  emitterUpdateScript [
   &EmitterState(Life Cycle Mode=::NewEnumerator1, Loop Behavior=::NewEnumerator1, Loop Duration Mode=::NewEnumerator1, UseLoopDelay=false)
   &SpawnBurst_Instantaneous(Spawn Count=2, Use Spawn Probability=false, UseLoopCountLimit=false)
  ]
  particleSpawnScript [
   &InitializeParticle(Lifetime=5, Mesh Scale=<0.3 0.15 0.15>, Mesh Scale Mode=::NewEnumerator4, UsePositionOffset=false, Write Material Random=true)
   &InitialMeshOrientation(Mesh Orientation Mode=::NewEnumerator2, Rotation=<0 0.25 0>, Use Rotation Vector=true)
   &DynamicMaterialParameters(
    Index 0 Param 1=1
    Index 0 Param 2=2
    Module.Index0Float0Write=true
    Module.Index0Float1Write=true
    Module.Index0Float2Write=true
    Module.Index0Float3Write=true
   )
  ]
  particleUpdateScript [
   &ParticleState(Kill Particles When Lifetime Has Elapsed=false)
   &SkeletalMeshLocation(
    Bone / Socket Index=&SelectIntFromArray(Direct Array Index=&ReturnExecIndex(), Int Selection Array=DataInterface() {propertyValues {IntData [5,6]}})
    Bone Sampling Mode=::NewEnumerator4
    MaskSpawnGroup=false
    Orientation Sampling=::NewEnumerator1
    Sampled Position Offset=&SelectVectorFromArray(
     Direct Array Index=&ReturnExecIndex()
     Vector Selection Array=DataInterface() {propertyValues {FloatData [<0 0 -10>, <0 0 -10>]}}
    )
    Sampled Position Offset Coordinate Space:ENiagaraSkelMeshTransforms=::NewEnumerator0
    Skeletal Mesh=$System.Skeletal Mesh
   )
   &UpdateMeshOrientation(
    Facing Interpolation=false
    Orientation Method=::NewEnumerator1
    Override Quat=false
    TransformFacing=false
    TransformUp=false
    Up Direction:Vector3f=&FN_FrontEndVector3Modifier(Gameplay=<0 0 1>, Front End=&Lerp_Vector(A=<0 -0.2 1>, B=<0 -0.7 1>, Alpha=$User.ParaGlideLeanAlpha))
   )
   &ScaleColor(
    Scale Alpha=&FN_FrontEndFloatModifier(
     Gameplay=&FNTimeOfDayScalar(Night Scale=2, Morning Scale=1.3, Day Scale=1.3, Evening Scale=1.3)
     Front End=1
    )
    Scale RGB=&MakeVectorFromLinearColorRGB(Color=$System.ToD Color)
    ScaleA=true
    ScaleRGB=true
   )
  ]
  rendererClasses [&NiagaraMeshRendererProperties]
  renderers [{rendererClass=&NiagaraMeshRendererProperties, rendererIndex=0}]
 }
 WindMeshDive {
  emitterUpdateScript [
   &EmitterState(Life Cycle Mode=::NewEnumerator1, Loop Behavior=::NewEnumerator1, Loop Duration Mode=::NewEnumerator1, UseLoopDelay=false)
   &SpawnBurst_Instantaneous(Use Spawn Probability=false, UseLoopCountLimit=false)
  ]
  particleSpawnScript [
   &InitializeParticle(
    Color=$System.ToD Color
    Lifetime=5
    Mesh Scale=<0.5 0.55 0.55>
    Mesh Scale Mode=::NewEnumerator4
    UsePositionOffset=false
    Write Material Random=true
   )
   &InitialMeshOrientation(Mesh Orientation Mode=::NewEnumerator2, Rotation=<0 0.25 0>, Use Rotation Vector=true)
   &DynamicMaterialParameters(
    Index 0 Param 1=2
    Index 0 Param 2=2
    Module.Index0Float0Write=true
    Module.Index0Float1Write=true
    Module.Index0Float2Write=true
    Module.Index0Float3Write=true
   )
  ]
  particleUpdateScript [
   &ParticleState(Kill Particles When Lifetime Has Elapsed=false)
   &SkeletalMeshLocation(
    Bone / Socket Index=4
    MaskSpawnGroup=false
    Mesh Sampling Type=::NewEnumerator5
    Orientation Sampling=::NewEnumerator1
    Sampled Position Offset=<0 0 -80>
    Sampled Position Offset Coordinate Space:ENiagaraSkelMeshTransforms=::NewEnumerator0
    Skeletal Mesh=$System.Skeletal Mesh
    Socket Sampling Mode=::NewEnumerator5
   )
   &UpdateMeshOrientation(
    Facing Interpolation=false
    Orientation Method=::NewEnumerator1
    Override Quat=false
    TransformFacing=false
    TransformUp=false
    Up Direction:Vector3f=&FN_FrontEndVector3Modifier(Gameplay=<0 0 1>, Front End=&Lerp_Vector(A=<0 -0.2 1>, B=<0 -0.7 1>, Alpha=$User.ParaGlideLeanAlpha))
   )
   &ScaleColor(
    Scale Alpha=&FN_FrontEndFloatModifier(
     Gameplay=&Multiply_Float(
      A=&FNTimeOfDayScalar(Night Scale=2, Morning Scale=1.3, Day Scale=1.3, Evening Scale=1.3)
      B=$User.ParaGlideLeanAlpha
     )
     Front End=0
    )
    ScaleA=true
    ScaleRGB=false
   )
  ]
  rendererClasses [&NiagaraMeshRendererProperties]
  renderers [{rendererClass=&NiagaraMeshRendererProperties, rendererIndex=0}]
 }
 WindTrails {
  emitterUpdateScript [
   &EmitterState()
   &SpawnBurst_Instantaneous(Spawn Count=40, Use Spawn Probability=false, UseLoopCountLimit=true)
  ]
  particleSpawnScript [
   &PartitionParticles(
    Normalized Index In Partition Includes One=true
    Number Of Partitions=4
    Particles Per Partition=10
    Partition Mode:ENiagara_PartitionMode=::NewEnumerator1
    Write Partition Index to Ribbon Link Order:ENiagara_AttributeSamplingApplyOutput=::NewEnumerator0
    Write Partition to Ribbon ID=::NewEnumerator0
   )
   &InitializeParticle(
    Color Mode=::NewEnumerator0
    Lifetime=&SelectFloatFromArray(
     Direct Array Index=$Particles.PartitionParticles.Partition
     Float Selection Array=DataInterface() {propertyValues {FloatData [0.4,0.4,0.125,0.125]}}
    )
    Ribbon Width=&SelectFloatFromArray(Direct Array Index=$Particles.PartitionParticles.Partition, Float Selection Array=DataInterface() {propertyValues {FloatData [20,20,13,13]}})
    Ribbon Width Mode=::NewEnumerator1
    Sprite Size Mode=::NewEnumerator3
    UsePositionOffset=false
    Write Material Random=true
   )
  ]
  particleUpdateScript [
   &ParticleState(Kill Particles When Lifetime Has Elapsed=false)
   &SkeletalMeshLocation(
    Bone / Socket Index=$Particles.PartitionParticles.Partition
    MaskSpawnGroup=false
    Mesh Sampling Type=::NewEnumerator5
    Skeletal Mesh=$System.Skeletal Mesh
    Socket Sampling Mode=::NewEnumerator5
   )
   /SetVariables(Particles.NewAge:NiagaraFloat=&Multiply_Float(A=$Particles.PartitionParticles.Normalized Index In Partition, B=$Particles.Initial.Lifetime)) {isSetParameters=true}
   /SetVariables(Particles.Velocity=&Multiply_VectorByFloat(
     Vector=&Lerp_Vector(A=$System.FN_Contrail_Setup.Direction Velocity Raw, B=$System.FN_Contrail_Setup.Direction Velocity, Alpha=$Particles.NewAge)
     Float=$Particles.Initial.Lifetime
    )) {isSetParameters=true}
   /SetVariables(Particles.Position=&AddVectorToPosition(
     Position=$Particles.Position
     Vector=&Multiply_VectorByFloat(Vector=$Particles.Velocity, Float=$Particles.PartitionParticles.Normalized Index In Partition)
    )) {isSetParameters=true}
   &ScaleRibbonWidth(Ribbon Width Scale=&FloatFromCurve(
     FloatCurve=DataInterface() {propertyValues {Curve {keys [{interpMode=RCIM_Cubic, leaveTangent=1.4222, tangentMode=RCTM_Break, time=0, value=0.8}, <at 1 1.7>]}}}
     CurveIndex=$Particles.PartitionParticles.Normalized Index In Partition
    )) {enabled=false}
   &ScaleColor(
    Scale Alpha=&Multiply_Float(
     A=&FN_FrontEndFloatModifier(Gameplay=0.4, Front End=0.12)
     B=$System.FN_Contrail_Setup.Normalized Intro Age
    )
    Scale RGB=&MakeVectorFromLinearColorRGB(Color=$System.ToD Color)
    ScaleA=true
    ScaleRGB=true
   )
   /SetVariables(
    Particles.FillTrailColor:LinearColor=&MakeLinearColorFromVectorAndFloat(
     "Vector (RGB)"=&MakeVectorFromLinearColorRGB(Color=$System.ToD Color)
     "Float (Alpha)"=&Multiply_Float(A=0.4, B=&OneMinusFloat(Float=$Particles.PartitionParticles.Normalized Index In Partition))
    )
    Particles.FillTrailDynPar:Vector4f=&MakeVector4(
     X=-2
     Y=0
     Z=1
     W=&SelectFloatFromArray(Direct Array Index=$Particles.PartitionParticles.Partition, Float Selection Array=DataInterface() {propertyValues {FloatData [2,2,1,1]}})
    )
   ) {isSetParameters=true}
  ]
  rendererClasses [&NiagaraSpriteRendererProperties, &NiagaraRibbonRendererProperties]
  renderers [
   {rendererClass=&NiagaraSpriteRendererProperties}
   {rendererClass=&NiagaraRibbonRendererProperties, rendererIndex=1}
   {rendererClass=&NiagaraRibbonRendererProperties, rendererIndex=2}
  ]
 }
}
summary {
 systemName=NS_Contrail_SquidGlisten
 emitters [
  {emitterName=Trails, rendererClasses [&NiagaraSpriteRendererProperties, &NiagaraRibbonRendererProperties]}
  {emitterName=WindTrails, rendererClasses [&NiagaraSpriteRendererProperties, &NiagaraRibbonRendererProperties]}
  {emitterName=WindMeshDive, rendererClasses [&NiagaraMeshRendererProperties]}
  {emitterName=Birdarangs, rendererClasses [&NiagaraSpriteRendererProperties, &NiagaraMeshRendererProperties]}
  {emitterName=WindMesh001, rendererClasses [&NiagaraMeshRendererProperties]}
 ]
 userVariables [
  User.bIsFrontEnd {type=NiagaraBool}
  User.bIsFrontEndPreview {type=NiagaraBool}
  User.ParaGlideLeanAlpha {type=NiagaraFloat, defaultValue=0, description=""}
  User.RibbonIDOffset {type=NiagaraInt32}
 ]
}
